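# Sound cue for the fire burst of the Nitro chimney stack environment actor, with randomized pitch over three mixed layers.
# Close and distant layers use the outdoor ambient class, each with its own attenuation and random wave pick; the distant randomizer has one empty slot.
# A quieter, randomly delayed third layer replays distant waves through the indoor class and a medium interior/exterior occlusion attenuation.
# Overridden concurrency caps voices per owner.
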SoundCue /Game/Sounds/Helios/Environment_Actors/NitroChimneyStack/Spot_Nitro_Chimney_Stack_Fire_Burst_Cue
root=Modulator_0
nodes [
 {
  id=Attenuation_1
  children [Random_2]
  properties {attenuationSettings=&/Nitro_Chimney_Fire_Burst_Medium_IntExt_Occlusion_Att}
 }
 {
  id=Attenuation_3
  children [Random_1]
  properties {attenuationSettings=&/Nitro_Chimney_Fire_Burst_Distant_Max_Focus_Att}
 }
 {
  id=Attenuation_4
  children [Random_0]
  properties {attenuationSettings=&/Nitro_Chimney_Fire_Burst_Close_Att}
 }
 {id=Delay_0, children [SoundClass_0], properties {delayMax=0.5, delayMin=0.2}}
 {id=Mixer_1, children [SoundClass_2,SoundClass_1,Delay_0], properties {inputVolume [1,2,0.5]}}
 {id=Modulator_0, children [Mixer_1], properties {pitchMax=1.1, pitchMin=0.8}}
 {
  id=Random_0
  children [WavePlayer_0,WavePlayer_3,WavePlayer_4,WavePlayer_5,WavePlayer_6,WavePlayer_7,WavePlayer_8,WavePlayer_9]
  properties {weights [1,1,1,1,1,1,1,1]}
 }
 {
  id=Random_1
  children [null, WavePlayer_11, WavePlayer_12, WavePlayer_13, WavePlayer_14, WavePlayer_15, WavePlayer_16, WavePlayer_17]
  properties {weights [1,1,1,1,1,1,1,1]}
 }
 {
  id=Random_2
  children [WavePlayer_1, WavePlayer_2, WavePlayer_18, WavePlayer_19, WavePlayer_20, WavePlayer_21, WavePlayer_22, WavePlayer_23]
  properties {weights [1,1,1,1,1,1,1,1]}
 }
 {id=SoundClass_0, children [Attenuation_1], properties {soundClassOverride=&Fort_Ambient_Indoors_Athena}}
 {
  id=SoundClass_1
  children [Attenuation_3]
  properties {soundClassOverride=&Fort_Ambient_Outdoors_Athena}
 }
 {
  id=SoundClass_2
  children [Attenuation_4]
  properties {soundClassOverride=&Fort_Ambient_Outdoors_Athena}
 }
 {
  id=WavePlayer_0
  properties {soundWaveAssetPtr=&/Spot_Nitro_Chimney_Fire_Burst_Close_01}
 }
 {
  id=WavePlayer_1
  properties {soundWaveAssetPtr=&/Spot_Nitro_Chimney_Fire_Burst_Distant_01}
 }
 {
  id=WavePlayer_11
  properties {soundWaveAssetPtr=&/Spot_Nitro_Chimney_Fire_Burst_Distant_02}
 }
 {
  id=WavePlayer_12
  properties {soundWaveAssetPtr=&/Spot_Nitro_Chimney_Fire_Burst_Distant_03}
 }
 {
  id=WavePlayer_13
  properties {soundWaveAssetPtr=&/Spot_Nitro_Chimney_Fire_Burst_Distant_04}
 }
 {
  id=WavePlayer_14
  properties {soundWaveAssetPtr=&/Spot_Nitro_Chimney_Fire_Burst_Distant_05}
 }
 {
  id=WavePlayer_15
  properties {soundWaveAssetPtr=&/Spot_Nitro_Chimney_Fire_Burst_Distant_06}
 }
 {
  id=WavePlayer_16
  properties {soundWaveAssetPtr=&/Spot_Nitro_Chimney_Fire_Burst_Distant_07}
 }
 {
  id=WavePlayer_17
  properties {soundWaveAssetPtr=&/Spot_Nitro_Chimney_Fire_Burst_Distant_08}
 }
 {
  id=WavePlayer_18
  properties {soundWaveAssetPtr=&Spot_Nitro_Chimney_Fire_Burst_Distant_03}
 }
 {
  id=WavePlayer_19
  properties {soundWaveAssetPtr=&Spot_Nitro_Chimney_Fire_Burst_Distant_04}
 }
 {
  id=WavePlayer_2
  properties {soundWaveAssetPtr=&Spot_Nitro_Chimney_Fire_Burst_Distant_02}
 }
 {
  id=WavePlayer_20
  properties {soundWaveAssetPtr=&Spot_Nitro_Chimney_Fire_Burst_Distant_05}
 }
 {
  id=WavePlayer_21
  properties {soundWaveAssetPtr=&Spot_Nitro_Chimney_Fire_Burst_Distant_06}
 }
 {
  id=WavePlayer_22
  properties {soundWaveAssetPtr=&Spot_Nitro_Chimney_Fire_Burst_Distant_07}
 }
 {
  id=WavePlayer_23
  properties {soundWaveAssetPtr=&Spot_Nitro_Chimney_Fire_Burst_Distant_08}
 }
 {
  id=WavePlayer_3
  properties {soundWaveAssetPtr=&/Spot_Nitro_Chimney_Fire_Burst_Close_02}
 }
 {
  id=WavePlayer_4
  properties {soundWaveAssetPtr=&/Spot_Nitro_Chimney_Fire_Burst_Close_03}
 }
 {
  id=WavePlayer_5
  properties {soundWaveAssetPtr=&/Spot_Nitro_Chimney_Fire_Burst_Close_04}
 }
 {
  id=WavePlayer_6
  properties {soundWaveAssetPtr=&/Spot_Nitro_Chimney_Fire_Burst_Close_05}
 }
 {
  id=WavePlayer_7
  properties {soundWaveAssetPtr=&/Spot_Nitro_Chimney_Fire_Burst_Close_06}
 }
 {
  id=WavePlayer_8
  properties {soundWaveAssetPtr=&/Spot_Nitro_Chimney_Fire_Burst_Close_07}
 }
 {
  id=WavePlayer_9
  properties {soundWaveAssetPtr=&/Spot_Nitro_Chimney_Fire_Burst_Close_08}
 }
]
properties {
 bOverrideConcurrency=true
 maxDistance=30000
 soundClassObject=&Fort_Ambient_Outdoors_Athena
 subtitlePriority=5000
 volumeMultiplier=0.251
 attenuationOverrides {
  attenuationShape=Capsule
  bAttenuateWithLPF=true
  bEnableFocusInterpolation=true
  bEnableListenerFocus=true
  dBAttenuationAtMax=-40
  falloffDistance=1000
  focusAzimuth=15
  focusVolumeAttenuation=1.5
  lPFFrequencyAtMax=1000
  lPFRadiusMax=3000
  lPFRadiusMin=1000
  nonFocusAzimuth=45
  nonFocusVolumeAttenuation=0.75
  occlusionInterpolationTime=1
  occlusionLowPassFilterFrequency=10000
  occlusionTraceChannel=ECC_GameTraceChannel11
  occlusionVolumeAttenuation=0.5
  stereoSpread=512
  attenuationShapeExtents=<1000 500 90>
 }
 concurrencyOverrides {bLimitToOwner=true, maxCount=6, voiceStealReleaseTime=2}
}
