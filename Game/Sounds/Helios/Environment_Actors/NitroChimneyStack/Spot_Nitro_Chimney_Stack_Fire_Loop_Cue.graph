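# Looping outdoor ambient cue for the fire of a Nitro chimney stack environment actor.
# Close and distant fire loops, each with its own attenuation, feed one mixer, then a modulator that randomizes pitch slightly.
# Two nested, endlessly looping envelopers of different lengths add slow pitch and volume variation so the loop drifts over time.
# A concurrency override caps simultaneous voices per owner.
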SoundCue /Game/Sounds/Helios/Environment_Actors/NitroChimneyStack/Spot_Nitro_Chimney_Stack_Fire_Loop_Cue
root=Enveloper_3
nodes [
 {
  id=Attenuation_3
  children [WavePlayer_1]
  properties {attenuationSettings=&/Nitro_Chimney_Fire_Distant_Max_Focus_Att}
 }
 {
  id=Attenuation_4
  children [WavePlayer_2]
  properties {attenuationSettings=&/Nitro_Chimney_Fire_Close_Att}
 }
 {
  id=Enveloper_0
  children [Modulator_0]
  properties {
   bLoop=true
   bLoopIndefinitely=true
   loopEnd=30
   pitchCurve {editorCurveData {keys [
      {arriveTangent=-0.0013, interpMode=RCIM_Cubic, leaveTangent=-0.0013, time=0}
      <at 15 0.98 RCIM_Cubic>
      {arriveTangent=0.0013, interpMode=RCIM_Cubic, leaveTangent=0.0013, time=30}
     ]}}
   volumeCurve {editorCurveData {keys [<at 0 1>]}}
  }
 }
 {
  id=Enveloper_3
  children [Enveloper_0]
  properties {
   bLoop=true
   bLoopIndefinitely=true
   loopEnd=20
   pitchCurve {editorCurveData {keys [{time=0}]}}
   volumeCurve {editorCurveData {keys [
      {arriveTangent=-0.01, interpMode=RCIM_Cubic, leaveTangent=-0.01, time=0, value=1}
      <at 10 0.9 RCIM_Cubic>
      {arriveTangent=0.01, interpMode=RCIM_Cubic, leaveTangent=0.01, time=20, value=1}
     ]}}
  }
 }
 {id=Mixer_1, children [Attenuation_4, Attenuation_3], properties {inputVolume [1,1]}}
 {id=Modulator_0, children [Mixer_1], properties {pitchMax=1.1, pitchMin=0.9}}
 {
  id=WavePlayer_1
  properties {bLooping=true, soundWaveAssetPtr=&/Spot_Nitro_Chimney_Fire_Distant_Loop}
 }
 {
  id=WavePlayer_2
  properties {bLooping=true, soundWaveAssetPtr=&/Spot_Nitro_Chimney_Fire_Loop}
 }
]
properties {
 bOverrideConcurrency=true
 maxDistance=11570.7998
 soundClassObject=&Fort_Ambient_Outdoors_Athena
 subtitlePriority=5000
 volumeMultiplier=0.501
 attenuationOverrides {
  attenuationShape=Capsule
  bAttenuateWithLPF=true
  bEnableFocusInterpolation=true
  bEnableListenerFocus=true
  dBAttenuationAtMax=-40
  falloffDistance=1000
  focusAzimuth=15
  focusVolumeAttenuation=1.5
  lPFFrequencyAtMax=1000
  lPFRadiusMax=3000
  lPFRadiusMin=1000
  nonFocusAzimuth=45
  nonFocusVolumeAttenuation=0.75
  occlusionInterpolationTime=1
  occlusionLowPassFilterFrequency=10000
  occlusionTraceChannel=ECC_GameTraceChannel11
  occlusionVolumeAttenuation=0.5
  stereoSpread=512
  attenuationShapeExtents=<1000 500 90>
 }
 concurrencyOverrides {bLimitToOwner=true, maxCount=2, voiceStealReleaseTime=2}
}
