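# Looping outdoor ambient sound cue for smoke from the Nitro chimney stack.
# One looping wave feeds close and distant-focus attenuation branches, mixed with the distant one quieter, then passes through a random pitch modulator.
# Two nested, endlessly looping envelopers with different periods apply slow pitch and volume curves to vary the loop.
# The cue overrides concurrency to a single voice and sends to the exterior ambient audio bus.
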SoundCue /Game/Sounds/Helios/Environment_Actors/NitroChimneyStack/Spot_Nitro_Chimney_Stack_Smoke_Loop_Cue
root=Enveloper_3
nodes [
 {
  id=Attenuation_2
  children [WavePlayer_2]
  properties {attenuationSettings=&/Nitro_Chimney_Smoke_Close_Att}
 }
 {
  id=Attenuation_3
  children [WavePlayer_2]
  properties {attenuationSettings=&/Nitro_Chimney_Smoke_Distant_Max_Focus_Att}
 }
 {
  id=Enveloper_0
  children [Modulator_0]
  properties {
   bLoop=true
   bLoopIndefinitely=true
   loopEnd=30
   pitchCurve {editorCurveData {keys [
      {arriveTangent=-0.0013, interpMode=RCIM_Cubic, leaveTangent=-0.0013, time=0}
      <at 15 0.98 RCIM_Cubic>
      {arriveTangent=0.0013, interpMode=RCIM_Cubic, leaveTangent=0.0013, time=30}
     ]}}
   volumeCurve {editorCurveData {keys [<at 0 1>]}}
  }
 }
 {
  id=Enveloper_3
  children [Enveloper_0]
  properties {
   bLoop=true
   bLoopIndefinitely=true
   loopEnd=20
   pitchCurve {editorCurveData {keys [{time=0}]}}
   volumeCurve {editorCurveData {keys [
      {arriveTangent=-0.01, interpMode=RCIM_Cubic, leaveTangent=-0.01, time=0, value=1}
      <at 10 0.9 RCIM_Cubic>
      {arriveTangent=0.01, interpMode=RCIM_Cubic, leaveTangent=0.01, time=20, value=1}
     ]}}
  }
 }
 {id=Mixer_2, children [Attenuation_2, Attenuation_3], properties {inputVolume [1,0.4]}}
 {id=Modulator_0, children [Mixer_2], properties {pitchMax=1.1, pitchMin=0.9}}
 {
  id=WavePlayer_2
  properties {bLooping=true, soundWaveAssetPtr=&/Spot_Nitro_Chimney_Smoke_Loop}
 }
]
properties {
 bOverrideConcurrency=true
 maxDistance=20000
 soundClassObject=&Fort_Ambient_Outdoors_Athena
 subtitlePriority=5000
 volumeMultiplier=0.2
 attenuationOverrides {
  attenuationShape=Capsule
  bAttenuateWithLPF=true
  bEnableFocusInterpolation=true
  bEnableListenerFocus=true
  dBAttenuationAtMax=-40
  falloffDistance=2638.2017
  lPFFrequencyAtMax=1000
  lPFRadiusMax=3000
  lPFRadiusMin=1000
  nonFocusVolumeAttenuation=0
  occlusionInterpolationTime=1
  occlusionLowPassFilterFrequency=10000
  occlusionTraceChannel=ECC_GameTraceChannel11
  occlusionVolumeAttenuation=0.5
  stereoSpread=512
  attenuationShapeExtents=<3720.1087 1500 90>
 }
 busSends [{audioBus=null, minSendLevel=0, sendLevel=1, soundSourceBus=&AmbientAudio_Exterior_SourceBus}]
 concurrencyOverrides {maxCount=1, voiceStealReleaseTime=2}
}
