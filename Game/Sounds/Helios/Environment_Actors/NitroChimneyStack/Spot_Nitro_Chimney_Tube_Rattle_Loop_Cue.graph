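# Ambient sound cue that loops a rattling tube sound for the Nitro chimney stack environment actor.
# A short random delay leads into an endlessly looping envelope that slowly varies volume and pitch over a randomly modulated looping wave.
# Only the first two quality-level slots connect to this chain; the others are empty.
# Overridden capsule attenuation with listener focus spatializes it, with limited concurrency and an exterior ambient bus send.
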SoundCue /Game/Sounds/Helios/Environment_Actors/NitroChimneyStack/Spot_Nitro_Chimney_Tube_Rattle_Loop_Cue
root=QualityLevel_0
nodes [
 {id=Delay_0, children [Enveloper_1], properties {delayMax=0.5}}
 {
  id=Enveloper_1
  children [Modulator_2]
  properties {
   bLoopIndefinitely=true
   loopEnd=30
   pitchCurve {editorCurveData {keys [{interpMode=RCIM_Cubic, time=0}, <at 15 1.1 RCIM_Cubic>, {interpMode=RCIM_Cubic, time=30}]}}
   volumeCurve {editorCurveData {keys [<at 0 1 RCIM_Cubic>,<at 15 0.7 RCIM_Cubic>,<at 30 1 RCIM_Cubic>]}}
  }
 }
 {id=Modulator_2, children [WavePlayer_2], properties {pitchMin=0.8}}
 {id=QualityLevel_0, children [Delay_0,Delay_0,null,null]}
 {
  id=WavePlayer_2
  properties {bLooping=true, soundWaveAssetPtr=&/Spot_Nitro_Chimney_Tube_Rattle}
 }
]
properties {
 attenuationSettings=/Game/Sounds/Helios/Environment_Actors/SewerPipes/Spot_UndergroundPipe_Loop_Capsule_Att
 bEnableSubmixSends=false
 bOverrideAttenuation=true
 bOverrideConcurrency=true
 maxDistance=1586.2073
 soundClassObject=&Fort_Ambient_Outdoors_Athena
 volumeMultiplier=0.301
 attenuationOverrides {
  attenuationShape=Capsule
  bEnableFocusInterpolation=true
  bEnableListenerFocus=true
  falloffDistance=1000
  focusAzimuth=15
  focusReleaseInterpSpeed=0.9
  focusVolumeAttenuation=1.2
  nonFocusAzimuth=45
  nonFocusVolumeAttenuation=0.4
  nonSpatializedRadiusEnd=125
  nonSpatializedRadiusStart=150
  attenuationShapeExtents=<586.2073 300>
 }
 busSends [{audioBus=null, minSendLevel=0.6, sendLevel=1, soundSourceBus=&AmbientAudio_Exterior_SourceBus}]
 concurrencyOverrides {maxCount=2, voiceStealReleaseTime=1}
 concurrencySet [&Weapon_Drop_StopOldest]
}
